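# Physics asset for the Ice King backpack cosmetic, defining collision bodies and joints for its cape.
# The root bone carries four overlapping capsules in a descending column, while left and right seven-bone cape chains use boxes, ending in tiny tip boxes.
# Each chain links neighboring bones with identical swing and tighter twist limits.
# No constraint joins the two chains to each other or to the root.
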PhysicsAsset /BRCosmetics/Accessories/FORT_Backpacks/M_MED_IceKing_Backpack/Meshes/M_MED_IceKing_Backpack_Physics
bodies {
 backpack_root [
  {shapeType=Capsule, length=35, radius=17.25, shapeName=backpack_root_capsule, center=<0 14.74 -24.7058>, rotation=<rot 0 90 90>}
  {shapeType=Capsule, length=35, radius=17.25, shapeName=backpack_root_capsule, center=<0 14.9078 -38.8206>, rotation=<rot 0 90 90>}
  {shapeType=Capsule, length=35, radius=17.25, shapeName=backpack_root_capsule, center=<0 14.6977 -54.2266>, rotation=<rot 0 90 90>}
  {shapeType=Capsule, length=35, radius=17.25, shapeName=backpack_root_capsule, center=<0 14.2828 -68.8587>, rotation=<rot 0 90 90>}
 ]
 cape_01_l [{shapeType=Box, extentX=8.6753, extentY=14.9476, extentZ=15.9452, center=<0.2595 -1.816 -0.4371>, rotation=<rot 0 -92.6934 -81.9833>}]
 cape_01_r [{shapeType=Box, extentX=14.9868, extentY=8.6769, extentZ=15.7736, center=<0.1579 -0.4356 -1.8073>, rotation=<rot 0 -93.2485 -87.5165>}]
 cape_02_l [{shapeType=Box, extentX=2.5269, extentY=8.9123, extentZ=14.2273, center=<7.0389 0.395 -0.0842>, rotation=<rot 0 -89.4467 -88.3513>}]
 cape_02_r [{shapeType=Box, extentX=2.4907, extentY=9.0341, extentZ=14.6045, center=<6.9393 0.3606 0.0637>, rotation=<rot 0 90.0831 -87.6476>}]
 cape_03_l [{shapeType=Box, extentX=2.0645, extentY=8.9165, extentZ=14.2309, center=<7.0149 0.3296 -0.2317>, rotation=<rot 0 91.041 -88.8727>}]
 cape_03_r [{shapeType=Box, extentX=1.8923, extentY=8.9185, extentZ=13.9186, center=<6.9761 0.3728 0.0746>, rotation=<rot 0 89.7294 -89.8267>}]
 cape_04_l [{shapeType=Box, extentX=1.7975, extentY=13.0885, extentZ=9.3948, center=<6.9838 0.4518 -0.13>, rotation=<rot -1.2308 -90 -0.8539>}]
 cape_04_r [{shapeType=Box, extentX=1.7975, extentY=13.0885, extentZ=9.3948, center=<6.9836 0.4518 0.13>, rotation=<rot 1.2308 -90 0.8539>}]
 cape_05_l [{shapeType=Box, extentX=1.9295, extentY=10.9319, extentZ=14.5176, center=<7.0777 0.4065 -0.0686>, rotation=<rot 0 -89.7297 -89.7207>}]
 cape_05_r [{shapeType=Box, extentX=1.9295, extentY=10.9319, extentZ=14.5176, center=<7.0777 0.4065 0.0686>, rotation=<rot 0 90.2703 -89.7207>}]
 cape_06_l [{shapeType=Box, extentX=2.575, extentY=12.3711, extentZ=15.8424, center=<8.1568 0.0736 -0.058>, rotation=<rot 0 91.1084 -89.8763>}]
 cape_06_r [{shapeType=Box, extentX=2.3503, extentY=12.2487, extentZ=15.0508, center=<8.1604 0.0732 0.0565>, rotation=<rot 0 -88.8913 -89.673>}]
 cape_07_l [{shapeType=Box, extentX=1.01, extentY=1.01, extentZ=1.01, rotation=<rot 0 -90 0>}]
 cape_07_r [{shapeType=Box, extentX=1.01, extentY=1.01, extentZ=1.01, rotation=<rot 0 -90 0>}]
}
constraints [
 {bone1Name=cape_03_l, bone2Name=cape_02_l, swing1LimitDegrees=40, swing2LimitDegrees=40, twistLimitDegrees=10}
 {bone1Name=cape_04_l, bone2Name=cape_03_l, swing1LimitDegrees=40, swing2LimitDegrees=40, twistLimitDegrees=10}
 {bone1Name=cape_05_l, bone2Name=cape_04_l, swing1LimitDegrees=40, swing2LimitDegrees=40, twistLimitDegrees=10}
 {bone1Name=cape_06_l, bone2Name=cape_05_l, swing1LimitDegrees=40, swing2LimitDegrees=40, twistLimitDegrees=10}
 {bone1Name=cape_07_l, bone2Name=cape_06_l, swing1LimitDegrees=40, swing2LimitDegrees=40, twistLimitDegrees=10}
 {bone1Name=cape_07_r, bone2Name=cape_06_r, swing1LimitDegrees=40, swing2LimitDegrees=40, twistLimitDegrees=10}
 {bone1Name=cape_06_r, bone2Name=cape_05_r, swing1LimitDegrees=40, swing2LimitDegrees=40, twistLimitDegrees=10}
 {bone1Name=cape_05_r, bone2Name=cape_04_r, swing1LimitDegrees=40, swing2LimitDegrees=40, twistLimitDegrees=10}
 {bone1Name=cape_03_r, bone2Name=cape_02_r, swing1LimitDegrees=40, swing2LimitDegrees=40, twistLimitDegrees=10}
 {bone1Name=cape_04_r, bone2Name=cape_03_r, swing1LimitDegrees=40, swing2LimitDegrees=40, twistLimitDegrees=10}
 {bone1Name=cape_02_l, bone2Name=cape_01_l, swing1LimitDegrees=40, swing2LimitDegrees=40, twistLimitDegrees=10}
 {bone1Name=cape_02_r, bone2Name=cape_01_r, swing1LimitDegrees=40, swing2LimitDegrees=40, twistLimitDegrees=10}
]
