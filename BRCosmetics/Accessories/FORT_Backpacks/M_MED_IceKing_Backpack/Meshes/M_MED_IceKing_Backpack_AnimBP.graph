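# Animation Blueprint for the Ice King backpack, animating its two cape chains.
# A cached input pose, swapped for a crouch pose asset and adjusted by spring bones, feeds RigidBody simulation when enabled, otherwise one of two trail-controller setups.
# Crouching or parachuting selects shorter trail chains; the default uses sine-driven fake velocity as wind.
# Event Graph maps crouch, skydive, parachute and backward movement to alphas, gravity and relaxation speeds, with bone rotations adjusting airborne poses.
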AnimBlueprint /BRCosmetics/Accessories/FORT_Backpacks/M_MED_IceKing_Backpack/Meshes/M_MED_IceKing_Backpack_AnimBP
target_skeleton=&/M_MED_IceKing_Backpack_Skeleton
graphs [
 AnimGraph {nodes [
   n1 = Get trail_alpha() -> (trail_alpha:double)
   n2 = Get fake_velocity() -> (fake_velocity:Vector)
   n3 = Trail controller - Bone: cape_07_l!(
    ComponentPose=Local To Component(LocalPose=Use cached pose 'subgraph_cache'())
    Alpha=@n1
    RelaxationSpeedScale:float=Get trail_relax_strength_left() -> (trail_relax_strength_left:double)
    FakeVelocity:Vector=@n2
   ) -> (Pose) {
    bActorSpaceFakeVel=true
    bShowTrailLocation=true
    bUsePlanarLimit=true
    chainLength=6
    debugLifeTime=30
    maxDeltaTime=0.0333
    alphaScaleBiasClamp {bInterpResult=true}
    baseJoint {boneName=backpack_root}
    planarLimits [{drivingBone {boneName=None}, planeTransform {location=<0 3 0>, rotation=<rot 0 0 -65>, scale=<1 1 1>}}]
    rotationLimits [
     {limitMax=<60 60 60>, limitMin=<-60 -60 -60>}
     {limitMax=<60 60 60>, limitMin=<-60 -60 -60>}
     {limitMax=<60 60 60>, limitMin=<-60 -60 -60>}
     {limitMax=<60 60 60>, limitMin=<-60 -60 -60>}
     {limitMax=<60 60 60>, limitMin=<-60 -60 -60>}
     {limitMax=<60 60 60>, limitMin=<-60 -60 -60>}
    ]
    rotationOffsets [<0 0 0>, <0 0 0>, <0 0 0>, <0 0 0>, <0 0 0>, <0 0 0>]
    trailBone {boneName=cape_07_l}
    trailRelaxationSpeed {editorCurveData {keys [<at 0 15 RCIM_Linear>,<at 0.5289 13.1838 RCIM_Linear>,<at 1 6 RCIM_Linear>]}}
   }
   n4 = Trail controller - Bone: cape_07_r!(
    ComponentPose=@n3
    Alpha=@n1
    RelaxationSpeedScale:float=Get trail_relax_strength_right() -> (trail_relax_strength_right:double)
    FakeVelocity:Vector=@n2
   ) -> (Pose) {
    bActorSpaceFakeVel=true
    bShowTrailLocation=true
    bUsePlanarLimit=true
    chainLength=6
    debugLifeTime=30
    maxDeltaTime=0.0333
    alphaScaleBiasClamp {bInterpResult=true}
    baseJoint {boneName=backpack_root}
    planarLimits [{drivingBone {boneName=None}, planeTransform {location=<0 3 0>, rotation=<rot 0 0 -65>, scale=<1 1 1>}}]
    rotationLimits [
     {limitMax=<180 0 0>, limitMin=<-180 0 0>}
     {limitMax=<180 0 0>, limitMin=<-180 0 0>}
     {limitMax=<180 0 0>, limitMin=<-180 0 0>}
     {limitMax=<180 0 0>, limitMin=<-180 0 0>}
     {limitMax=<180 0 0>, limitMin=<-180 0 0>}
     {limitMax=<180 0 0>, limitMin=<-180 0 0>}
    ]
    rotationOffsets [<0 0 0>, <0 0 0>, <0 0 0>, <0 0 0>, <0 0 0>, <0 0 0>]
    trailBone {boneName=cape_07_r}
    trailRelaxationSpeed {editorCurveData {keys [<at 0 15 RCIM_Linear>,<at 0.516 13.407 RCIM_Linear>,<at 1 8 RCIM_Linear>]}}
   }
   n5 = Get Is SkyDiving Float() -> (Is SkyDiving Float:float)
   n6 = "Transform (Modify) Bone - Bone: cape_01_l"(ComponentPose=@n4, Alpha=@n5, Translation="0,0,0", Rotation="0,-15,0", Scale="1,1,1") {
    rotationMode=BMM_Replace
    rotationSpace=BCS_BoneSpace
    translationMode=BMM_Replace
    translationSpace=BCS_BoneSpace
    alphaBoolBlend {blendOption=Linear}
    alphaScaleBiasClamp {bInterpResult=true}
    boneToModify {boneName=cape_01_l}
    rotation {yaw=-15}
   }
   n7 = "Transform (Modify) Bone - Bone: cape_01_r"(ComponentPose=@n6, Alpha=@n5, Translation="0,0,0", Rotation="-15,0,0", Scale="1,1,1") {
    rotationMode=BMM_Replace
    rotationSpace=BCS_BoneSpace
    translationMode=BMM_Replace
    translationSpace=BCS_BoneSpace
    alphaBoolBlend {blendOption=Linear}
    alphaScaleBiasClamp {bInterpResult=true}
    boneToModify {boneName=cape_01_r}
    rotation {pitch=-15}
   }
   n8 = "Transform (Modify) Bone - Bone: cape_02_l"(ComponentPose=@n7, Alpha=@n5, Translation="0,0,0", Rotation="0,-5,0", Scale="1,1,1") {
    rotationMode=BMM_Replace
    rotationSpace=BCS_BoneSpace
    translationMode=BMM_Replace
    translationSpace=BCS_BoneSpace
    alphaBoolBlend {blendOption=Linear}
    alphaScaleBiasClamp {bInterpResult=true}
    boneToModify {boneName=cape_02_l}
    rotation {yaw=-5}
   }
   n9 = "Transform (Modify) Bone - Bone: cape_02_r"(ComponentPose=@n8, Alpha=@n5, Translation="0,0,0", Rotation="0,-5,0", Scale="1,1,1") {
    rotationMode=BMM_Replace
    rotationSpace=BCS_BoneSpace
    translationMode=BMM_Replace
    translationSpace=BCS_BoneSpace
    alphaBoolBlend {blendOption=Linear}
    alphaScaleBiasClamp {bInterpResult=true}
    boneToModify {boneName=cape_02_r}
    rotation {yaw=-5}
   }
   n10 = "Transform (Modify) Bone - Bone: cape_03_l"(ComponentPose=@n9, Alpha=@n5, Translation="0,0,0", Rotation="0,-5,0", Scale="1,1,1") {
    rotationMode=BMM_Replace
    rotationSpace=BCS_BoneSpace
    translationMode=BMM_Replace
    translationSpace=BCS_BoneSpace
    alphaBoolBlend {blendOption=Linear}
    alphaScaleBiasClamp {bInterpResult=true}
    boneToModify {boneName=cape_03_l}
    rotation {yaw=-5}
   }
   n11 = "Transform (Modify) Bone - Bone: cape_03_r"(ComponentPose=@n10, Alpha=@n5, Translation="0,0,0", Rotation="0,-5,0", Scale="1,1,1") {
    rotationMode=BMM_Replace
    rotationSpace=BCS_BoneSpace
    translationMode=BMM_Replace
    translationSpace=BCS_BoneSpace
    alphaBoolBlend {blendOption=Linear}
    alphaScaleBiasClamp {bInterpResult=true}
    boneToModify {boneName=cape_03_r}
    rotation {yaw=-5}
   }
   n12 = "Transform (Modify) Bone - Bone: cape_04_l"(ComponentPose=@n11, Alpha=@n5, Translation="0,0,0", Rotation="0,-5,0", Scale="1,1,1") {
    rotationMode=BMM_Replace
    rotationSpace=BCS_BoneSpace
    translationMode=BMM_Replace
    translationSpace=BCS_BoneSpace
    alphaBoolBlend {blendOption=Linear}
    alphaScaleBiasClamp {bInterpResult=true}
    boneToModify {boneName=cape_04_l}
    rotation {yaw=-5}
   }
   n13 = "Transform (Modify) Bone - Bone: cape_04_r"(ComponentPose=@n12, Alpha=@n5, Translation="0,0,0", Rotation="0,-5,0", Scale="1,1,1") {
    rotationMode=BMM_Replace
    rotationSpace=BCS_BoneSpace
    translationMode=BMM_Replace
    translationSpace=BCS_BoneSpace
    alphaBoolBlend {blendOption=Linear}
    alphaScaleBiasClamp {bInterpResult=true}
    boneToModify {boneName=cape_04_r}
    rotation {yaw=-5}
   }
   n14 = "Transform (Modify) Bone - Bone: cape_05_l"(ComponentPose=@n13, Alpha=@n5, Translation="0,0,0", Rotation="0,-5,0", Scale="1,1,1") {
    rotationMode=BMM_Replace
    rotationSpace=BCS_BoneSpace
    translationMode=BMM_Replace
    translationSpace=BCS_BoneSpace
    alphaBoolBlend {blendOption=Linear}
    alphaScaleBiasClamp {bInterpResult=true}
    boneToModify {boneName=cape_05_l}
    rotation {yaw=-5}
   }
   n15 = "Transform (Modify) Bone - Bone: cape_05_r"(ComponentPose=@n14, Alpha=@n5, Translation="0,0,0", Rotation="0,-5,0", Scale="1,1,1") {
    rotationMode=BMM_Replace
    rotationSpace=BCS_BoneSpace
    translationMode=BMM_Replace
    translationSpace=BCS_BoneSpace
    alphaBoolBlend {blendOption=Linear}
    alphaScaleBiasClamp {bInterpResult=true}
    boneToModify {boneName=cape_05_r}
    rotation {yaw=-5}
   }
   n16 = "Transform (Modify) Bone - Bone: cape_06_l"(ComponentPose=@n15, Alpha=@n5, Translation="0,0,0", Rotation="0,-5,0", Scale="1,1,1") {
    rotationMode=BMM_Replace
    rotationSpace=BCS_BoneSpace
    translationMode=BMM_Replace
    translationSpace=BCS_BoneSpace
    alphaBoolBlend {blendOption=Linear}
    alphaScaleBiasClamp {bInterpResult=true}
    boneToModify {boneName=cape_06_l}
    rotation {yaw=-5}
   }
   n17 = "Transform (Modify) Bone - Bone: cape_06_r"(ComponentPose=@n16, Alpha=@n5, Translation="0,0,0", Rotation="0,-5,0", Scale="1,1,1") {
    rotationMode=BMM_Replace
    rotationSpace=BCS_BoneSpace
    translationMode=BMM_Replace
    translationSpace=BCS_BoneSpace
    alphaBoolBlend {blendOption=Linear}
    alphaScaleBiasClamp {bInterpResult=true}
    boneToModify {boneName=cape_06_r}
    rotation {yaw=-5}
   }
   n18 = "Transform (Modify) Bone - Bone: cape_07_l"(ComponentPose=@n17, Alpha=@n5, Translation="0,0,0", Rotation="0,-5,0", Scale="1,1,1") {
    rotationMode=BMM_Replace
    rotationSpace=BCS_BoneSpace
    translationMode=BMM_Replace
    translationSpace=BCS_BoneSpace
    alphaBoolBlend {blendOption=Linear}
    alphaScaleBiasClamp {bInterpResult=true}
    boneToModify {boneName=cape_07_l}
    rotation {yaw=-5}
   }
   n19 = "Transform (Modify) Bone - Bone: cape_07_r"(ComponentPose=@n18, Alpha=@n5, Translation="0,0,0", Rotation="0,-5,0", Scale="1,1,1") {
    rotationMode=BMM_Replace
    rotationSpace=BCS_BoneSpace
    translationMode=BMM_Replace
    translationSpace=BCS_BoneSpace
    alphaBoolBlend {blendOption=Linear}
    alphaScaleBiasClamp {bInterpResult=true}
    boneToModify {boneName=cape_07_r}
    rotation {yaw=-5}
   }
   n20 = Component To Local(ComponentPose=@n19)
   n21 = Get parachuting_mult() -> (parachuting_mult:double)
   n22 = Get crouch_trail_alpha() -> (crouch_trail_alpha:double)
   n23 = Trail controller - Bone: cape_07_l(ComponentPose=Local To Component(LocalPose=Use cached pose 'subgraph_cache'()), Alpha=@n22, RelaxationSpeedScale=@n21, FakeVelocity="0,0,0") {
    bShowTrailLocation=true
    chainLength=4
    debugLifeTime=30
    maxDeltaTime=0.0333
    alphaScaleBiasClamp {bInterpResult=true}
    baseJoint {boneName=backpack_root}
    rotationLimits [{limitMax=<180 180 180>}, {limitMax=<180 180 180>}, {limitMax=<180 180 180>}, {limitMax=<180 180 180>}]
    rotationOffsets [<0 0 0>, <0 0 0>, <0 0 0>, <0 0 0>]
    trailBone {boneName=cape_07_l}
    trailRelaxationSpeed {editorCurveData {keys [<at 0 15 RCIM_Linear>,<at 1 6 RCIM_Linear>]}}
   }
   n24 = Trail controller - Bone: cape_07_r(ComponentPose=@n23, Alpha=@n22, RelaxationSpeedScale=@n21, FakeVelocity="0,0,0") {
    bShowTrailLocation=true
    chainLength=4
    debugLifeTime=30
    maxDeltaTime=0.0333
    alphaScaleBiasClamp {bInterpResult=true}
    baseJoint {boneName=backpack_root}
    rotationLimits [{limitMax=<180 180 180>}, {limitMax=<180 180 180>}, {limitMax=<180 180 180>}, {limitMax=<180 180 180>}]
    rotationOffsets [<0 0 0>, <0 0 0>, <0 0 0>, <0 0 0>]
    trailBone {boneName=cape_07_r}
    trailRelaxationSpeed {editorCurveData {keys [<at 0 15 RCIM_Linear>,<at 1 8 RCIM_Linear>]}}
   }
   n25 = Get Parachute Opened Alpha() -> (Parachute Opened Alpha:float)
   n26 = "Transform (Modify) Bone - Bone: cape_01_l"(ComponentPose=@n24, Alpha=@n25, Translation="0,0,0", Rotation="0,-15,0", Scale="1,1,1") {
    rotationMode=BMM_Replace
    rotationSpace=BCS_BoneSpace
    translationMode=BMM_Additive
    translationSpace=BCS_BoneSpace
    alphaBoolBlend {blendOption=Linear}
    boneToModify {boneName=cape_01_l}
    rotation {yaw=-15}
   }
   n27 = "Transform (Modify) Bone - Bone: cape_01_r"(ComponentPose=@n26, Alpha=@n25, Translation="0,0,0", Rotation="-15,0,0", Scale="1,1,1") {
    rotationMode=BMM_Replace
    rotationSpace=BCS_BoneSpace
    translationMode=BMM_Additive
    translationSpace=BCS_BoneSpace
    alphaBoolBlend {blendOption=Linear}
    boneToModify {boneName=cape_01_r}
    rotation {pitch=-15}
   }
   n28 = Component To Local(ComponentPose=@n27)
   n29 = Get is_crouched_or_parachute() -> (is_crouched_or_parachute:bool)
   n30 = RigidBody!(
    ComponentPose=Local To Component(LocalPose=Use cached pose 'subgraph_cache'())
    Alpha=Get Is Skydiving Alpha() -> (Is Skydiving Alpha:float)
    OverrideWorldGravity:Vector=Get Gravity Override()
    ExternalForce=Get char_direction() -> (char_direction:Vector)
   ) -> (Pose) {
    bClampLinearTranslationLimitToRefPose=true
    bOverrideWorldGravity=true
    overridePhysicsAsset=&/M_MED_IceKing_Backpack_Physics
    simulationSpace=BaseBoneSpace
    alphaBoolBlend {blendOption=Linear}
    baseBoneRef {boneName=backpack_root}
    componentAppliedLinearAccClamp=<3000 3000 3000>
    componentLinearAccScale=<0.4 0.3 0.3>
    componentLinearVelScale=<0.4 0.3 0.3>
    simSpaceSettings {
     maxAngularAcceleration=1000
     maxAngularVelocity=3
     maxLinearAcceleration=700
     maxLinearVelocity=500
     velocityScaleZ=0
     worldAlpha=0.5
     externalLinearDragV=<0.2 0.2>
    }
   }
   n31 = Get Parachute Opened Alpha()
   n32 = "Transform (Modify) Bone - Bone: cape_01_l"(ComponentPose=@n30, Alpha=@n31, Translation="0,0,0", Rotation="0,-15,0", Scale="1,1,1") {
    rotationMode=BMM_Additive
    rotationSpace=BCS_BoneSpace
    translationMode=BMM_Additive
    translationSpace=BCS_BoneSpace
    alphaBoolBlend {blendOption=Linear}
    boneToModify {boneName=cape_01_l}
    rotation {yaw=-15}
   }
   n33 = "Transform (Modify) Bone - Bone: cape_01_r"(ComponentPose=@n32, Alpha=@n31, Translation="0,0,0", Rotation="-15,0,0", Scale="1,1,1") {
    rotationMode=BMM_Additive
    rotationSpace=BCS_BoneSpace
    translationMode=BMM_Additive
    translationSpace=BCS_BoneSpace
    alphaBoolBlend {blendOption=Linear}
    boneToModify {boneName=cape_01_r}
    rotation {pitch=-15}
   }
   Output Pose(Result=Blend Poses by bool(
     BlendPose_0=Component To Local(ComponentPose=@n33)
     BlendPose_1=Blend Poses by bool(BlendPose_0=@n28, BlendPose_1=@n20, BlendTime_0=1, BlendTime_1=1, bActiveValue=@n29) {blendType=Linear}
     bActiveValue=Get bRigidBodyEnabled()
    ) {blendType=Linear}) {layerGroup=None}
   n34 = M_MED_IceKing_CrouchPose_PoseAsset : Pose_0!(PoseAsset:Pose Asset Object Reference=&/M_MED_IceKing_CrouchPose_PoseAsset) -> (Pose:Pose Link Structure) {is=title AnimGraphNode_PoseByName, poseName=Pose_0}
   AnimGraphNode_SubInput_0 = Input Pose()
   n35 = Get blend_back_pose() -> (blend_back_pose:bool)
   n36 = Get crouch_alpha() -> (crouch_alpha:double)
   n37 = Spring controller - Bone: cape_03_r!(
    ComponentPose:Component Space Pose Link Structure=Local To Component(LocalPose=Blend Poses by bool(BlendPose_0=@n34, BlendPose_1=@AnimGraphNode_SubInput_0, BlendTime_0=1, BlendTime_1=1, bActiveValue=@n35) {blendType=Linear})
    Alpha:float=@n36
   ) -> (Pose:Component Space Pose Link Structure) {
    is=title AnimGraphNode_SpringBone
    bLimitDisplacement=true
    bRotateX=true
    bRotateY=true
    bRotateZ=true
    maxDisplacement=2
    alphaBoolBlend {blendOption=Linear}
    springBone {boneName=cape_03_r}
   }
   n38 = Spring controller - Bone: cape_03_l!(ComponentPose:Component Space Pose Link Structure=@n37, Alpha:float=@n36) -> (Pose:Component Space Pose Link Structure) {
    is=title AnimGraphNode_SpringBone
    bLimitDisplacement=true
    bRotateX=true
    bRotateY=true
    bRotateZ=true
    maxDisplacement=2
    alphaBoolBlend {blendOption=Linear}
    springBone {boneName=cape_03_l}
   }
   Save cached pose 'subgraph_cache'(Pose=Component To Local(ComponentPose=@n38)) {cacheName=subgraph_cache}
  ]}
 EventGraph {nodes [
   Event BlueprintInitializeAnimation()
   n1 = TryGetPawnOwner()
   > Is Valid(InputObject=@n1)
   > n2 = Cast To PlayerPawn_Athena(Object=@n1)
   > Set PawnOwner(PawnOwner:Player Pawn Athena Object Reference=@n2) -> (Output_Get:Player Pawn Athena Object Reference)
   > Set random_stream(random_stream:Random Stream Structure=MakeRandomStream!(InitialSeed:int=0) -> (ReturnValue:Random Stream Structure)) -> (Output_Get:Random Stream Structure)
   Event BlueprintUpdateAnimation()
   > n3 = Sequence!(execute) -> (then_0, then_1, then_2:Exec, then_3:Exec, then_4:Exec)
   n4 = "Absolute (Float)"(A=GetOwnerDirection!(self:Custom Character Part Anim Instance Object Reference) -> (ReturnValue:float))
   n5 = Get bIsSkyDiving()
   > Set isMovingBackward(isMovingBackward:bool=AND Boolean(A=NOT Boolean(A=@n5), B="float > float"(A=@n4, B=130) {is=title K2Node_CallFunction})) -> (Output_Get:bool)
   n6 = Get bIsSkyDiving()
   > Set blend_back_pose(blend_back_pose:bool=AND Boolean(A=NOT Boolean(A=@n6), B=Get bIsCrouching())) -> (Output_Get:bool)
   n7 = Get bIsCrouching()
   n8 = SelectFloat(A=2.5, B=1, bPickA=Get bIsParachuteOpened())
   > Set trail_relax_strength_left(trail_relax_strength_left:double=float * float(A=SelectFloat(A=1.85, B=0.85, bPickA=@n7), B=@n8)) -> (Output_Get:double)
   > Set trail_relax_strength_right(trail_relax_strength_right:double=float * float(A=SelectFloat(A=2, B=1, bPickA=@n7), B=@n8)) -> (Output_Get:double)
   > Set crouch_alpha(crouch_alpha:double=SelectFloat(A=0, B=1, bPickA=@n7)) -> (Output_Get:double)
   n9 = Get isMovingBackward() -> (isMovingBackward:bool)
   > Set crouch_trail_alpha(crouch_trail_alpha:double=SelectFloat(A=0, B=0.4, bPickA=@n9)) -> (Output_Get:double)
   n10 = Get Is SkyDiving Float()
   > Set trail_alpha(trail_alpha:double=SelectFloat(A=0, B=float - float(A=1, B=@n10), bPickA=@n9)) -> (Output_Get:double)
   n11 = GetGameTimeInSeconds()
   n12 = SelectFloat(A=10, B=1, bPickA=Get bIsSkyDiving())
   n13 = float * float(A=float * float(A=@n11, B=25), B=@n12)
   n14 = float * float(A=float * float(A=@n11, B=250), B=@n12)
   n15 = float + float(A=float * float(A="Sin (Degrees)"!(A:double=@n14) -> (ReturnValue:double), B=20), B=float * float(A="Sin (Degrees)"(A=@n13), B=80)) {is=title K2Node_CommutativeAssociativeBinaryOperator}
   > Set fake_velocity(fake_velocity:Vector=MakeVector(X="Absolute (Float)"(A=@n15), Y=0)) -> (Output_Get:Vector)
   > Set parachuting_mult(parachuting_mult:double=SelectFloat(A=4, B=1, bPickA=Get bIsParachuteOpened())) -> (Output_Get:double)
   > Set is_crouched_or_parachute(is_crouched_or_parachute:bool=OR Boolean(A=Get bIsParachuteOpened(), B=Get bIsCrouching())) -> (Output_Get:bool)
   n16 = Gravity Override State Function DEPRECATED(
    execute=@n3.then_1
    CharacterPartAnimInstance=Self-Reference()
    GameTimeInSeconds=GetGameTimeInSeconds()
    GravityOverrideParams=Get Gravity Override Params()
   )
   > Set Gravity Override(Gravity Override=@n16)
   n17 = Branch(execute=@n3.then_2, Condition=Get bIsSkyDiving())
   > Set Is Skydiving Alpha(Is Skydiving Alpha:float=0.45) -> (Output_Get:float)
   Set Is Skydiving Alpha(execute=@n17.else, Is Skydiving Alpha=1)
   Set Parachute Opened Alpha(execute=@n3.then_3, Parachute Opened Alpha:float="To Float (Boolean)"(InBool=Get bIsParachuteOpened())) -> (Output_Get:float)
   Set Is SkyDiving Float(execute=@n3.then_4, Is SkyDiving Float:float="To Float (Boolean)"(InBool=Get bIsSkyDiving())) -> (Output_Get:float)
   RandomFloatInRangeFromStream!(Stream:"Random Stream Structure (by ref)"=Get random_stream() -> (random_stream:Random Stream Structure), Min:float=1, Max:float=100) -> (ReturnValue:float)
  ]}
]
variables [
 blend_back_pose {type=boolean}
 char_direction {type=Vector}
 crouch_alpha {type=number}
 crouch_trail_alpha {type=number}
 fake_velocity {type=Vector}
 Gravity Override {type=Vector}
 Gravity Override Params {type=GravityOverrideParamsStruct}
 Is Skydiving Alpha {type=number}
 Is SkyDiving Float {type=number}
 is_crouched_or_parachute {type=boolean}
 isMovingBackward {type=boolean}
 Parachute Opened Alpha {type=number}
 parachuting_mult {type=number}
 PawnOwner {type=PlayerPawn_Athena_C}
 random_stream {type=RandomStream, default {initialSeed=0, seed=0}}
 trail_alpha {type=number}
 trail_relax_strength_left {type=number}
 trail_relax_strength_right {type=number}
]
